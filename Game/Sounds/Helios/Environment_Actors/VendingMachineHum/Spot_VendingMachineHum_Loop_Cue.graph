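# Vending machine hum ambient cue for Helios, built from two oil rig refrigerator hum waves pitched down.
# A Random node, shared by all quality levels, picks a single pitched hum or one of two layered mixes at different pitches and volumes.
# Only the Hum_02 players are flagged looping.
# It uses a shared vending machine attenuation asset, an ambient props sound class, per-owner concurrency limits and a reverb send.
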SoundCue /Game/Sounds/Helios/Environment_Actors/VendingMachineHum/Spot_VendingMachineHum_Loop_Cue
root=Modulator_7
nodes [
 {id=Mixer_0, children [Modulator_4,Modulator_1], properties {inputVolume [0.5,0.5]}}
 {id=Mixer_1, children [Modulator_5,Modulator_6], properties {inputVolume [1,0.5]}}
 {id=Mixer_2, children [Mixer_1], properties {inputVolume [0.8]}}
 {id=Modulator_1, children [WavePlayer_0], properties {pitchMax=0.8, pitchMin=0.8}}
 {id=Modulator_4, children [WavePlayer_1], properties {pitchMax=0.8, pitchMin=0.7}}
 {id=Modulator_5, children [WavePlayer_3], properties {pitchMax=0.6, pitchMin=0.6}}
 {id=Modulator_6, children [WavePlayer_9], properties {pitchMax=0.7, pitchMin=0.7}}
 {id=Modulator_7, children [QualityLevel_0]}
 {id=QualityLevel_0, children [Random_0,Random_0,Random_0,Random_0]}
 {id=Random_0, children [Modulator_4,Mixer_0,Mixer_2], properties {weights [1,1,1]}}
 {
  id=WavePlayer_0
  properties {bLooping=true, soundWaveAssetPtr=/Game/Sounds/Asteria/Environment_Actors/OilRig_Refridgerator/Spot_OilRig_Refridgerator_Hum_02_Loop}
 }
 {
  id=WavePlayer_1
  properties {soundWaveAssetPtr=/Game/Sounds/Asteria/Environment_Actors/OilRig_Refridgerator/Spot_OilRig_Refridgerator_Hum_Loop}
 }
 {
  id=WavePlayer_3
  properties {soundWaveAssetPtr=&Spot_OilRig_Refridgerator_Hum_Loop}
 }
 {
  id=WavePlayer_9
  properties {bLooping=true, soundWaveAssetPtr=&Spot_OilRig_Refridgerator_Hum_02_Loop}
 }
]
properties {
 attenuationSettings=/Game/Sounds/Asteria/Environment_Actors/VendingMachine/Spot_VendingMachine_Hum_Loop_Att
 bOverrideConcurrency=true
 maxDistance=550
 soundClassObject=&Fort_Ambient_Props_Athena
 subtitlePriority=5000
 volumeMultiplier=0.4
 attenuationOverrides {
  bEnableFocusInterpolation=true
  bEnableListenerFocus=true
  bEnableOcclusion=true
  dBAttenuationAtMax=-40
  falloffDistance=512
  focusAzimuth=15
  focusVolumeAttenuation=1.5
  nonFocusAzimuth=45
  nonFocusVolumeAttenuation=0.5
  occlusionInterpolationTime=1
  occlusionLowPassFilterFrequency=10000
  occlusionTraceChannel=ECC_GameTraceChannel11
  occlusionVolumeAttenuation=0.5
  stereoSpread=512
  attenuationShapeExtents=<256 64.5942 90>
 }
 concurrencyOverrides {bLimitToOwner=true, maxCount=2}
 soundSubmixSends [{minSendLevel=1, soundSubmix=&Fortnite_Reverb_Default}]
}
